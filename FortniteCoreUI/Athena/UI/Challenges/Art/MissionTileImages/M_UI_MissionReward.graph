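# Translucent UI material for mission reward tiles in challenges, drawing a bordered box, a circular pip or a connector bar via static switches.
# Rotated UV edge thresholds form an irregular quad, radial gradients form circles; inner and outer masks separate accent border from fill.
# The box fill shows a preview image, unknown-mission icon or flat color, with a star scalar blending in the mission icon.
# Prestige and inactive scalars blend palette colors and alphas.
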Material /FortniteCoreUI/Athena/UI/Challenges/Art/MissionTileImages/M_UI_MissionReward
graph {
 properties {
  bUsedWithStaticMesh=false
  blendMode=BLEND_Translucent
  materialDomain=MD_UI
  pixelDepthOffsetMode=PDOM_Legacy
  usedShadingModels=Default Lit
 }
 nodes [
  n1 = LinearInterpolate(
   A=LinearInterpolate(
    A=VectorParameter() {parameterName=Accent, defaultValue=<rgba 1 0.4023 0.1599 1>}.RGB
    B=VectorParameter() {parameterName=PrestigeAccent, defaultValue=<rgba 1 0.855 0.0331 1>}.RGB
    Alpha=ScalarParameter() {parameterName=IsPrestige}
   )
   B=LinearInterpolate(
    A=VectorParameter() {parameterName=InactiveAccent, defaultValue=<rgba 0.3185 0.6038 1 1>}.RGB
    B=VectorParameter() {parameterName=PrestigeInactiveAccent, defaultValue=<rgba 0.9734 1 0.4452 1>}.RGB
    Alpha=ScalarParameter() {parameterName=IsPrestige}
   )
   Alpha=ScalarParameter() {parameterName=IsInactive}
  )
  n2 = LinearInterpolate(
   A=VectorParameter() {parameterName=Primary, defaultValue=<rgba 0.2344 0.113 0.056 1>}.RGB
   B=VectorParameter() {parameterName=PrimaryInactive, defaultValue=<rgba 0 0.004 0.0723 1>}.RGB
   Alpha=ScalarParameter() {parameterName=IsInactive}
  )
  n3 = MaterialFunctionCall(Texture Scale=ScalarParameter() {defaultValue=0.75, parameterName=Scale}) {materialFunction=&ScaleUVsByCenter}
  n4 = TextureSampleParameter2D(UVs=@n3.UVs) {
   parameterName=SmallPreviewImage
   samplerSource=SSM_Clamp_WorldGroupSettings
   texture=/Game/UI/Foundation/Textures/Icons/Heroes/Athena/Soldier/T-Soldier-HID-005-Athena-Commando-M
  }
  n5 = TextureSample(UVs=@n3.UVs) {texture=/Game/Athena/UI/Challenges/Art/MissionTileImages/T_UI_Icon_MissionUnknown}
  n6 = StaticSwitchParameter(
   True=@n2
   False=LinearInterpolate(
    A=LinearInterpolate(A=@n2, B=@n4.RGB, Alpha=@n4.RGB)
    B=LinearInterpolate(A=@n2, B=@n1, Alpha=@n5.R)
    Alpha=ScalarParameter() {parameterName=IsDefault}
   )
  ) {parameterName=ImgOff}
  n7 = ScalarParameter() {defaultValue=0.694, parameterName=Circle2}
  n8 = MaterialFunctionCall(
   UVs=TextureCoordinate()
   CenterPosition=AppendVector(A=ScalarParameter() {defaultValue=0.5, parameterName=CircleCenter}, B=Constant() {r=0.5})
  ) {materialFunction=&RadialGradientExponential}
  n9 = Clamp(None=Multiply(A=Subtract(A=@n8.RadialGradientExponential, B=@n7), B=ScalarParameter() {defaultValue=101, parameterName=Circle2_Sharp}))
  n10 = TextureCoordinate()
  n11 = ComponentMask(None=Rotator(Coordinate=@n10, Time=ScalarParameter() {defaultValue=0.0426, group=BoxShape, parameterName=U1-Rotation}) {speed=6.2832}) {r=true}
  n12 = ScalarParameter() {defaultValue=0.19, group=BoxShape, parameterName=U1-Offset-Inner}
  n13 = ScalarParameter() {defaultValue=128, group=BoxShape, parameterName=U1-Edge}
  n14 = Clamp(None=Multiply(A=Subtract(A=@n11, B=@n12), B=@n13))
  n15 = ScalarParameter() {defaultValue=0.013, group=BoxShape, parameterName=U2-Rotation}
  n16 = OneMinus(None=ComponentMask(None=Rotator(Coordinate=@n10, Time=@n15) {speed=6.2832}) {r=true})
  n17 = ScalarParameter() {defaultValue=0.19, group=BoxShape, parameterName=U2-Offset-Inner}
  n18 = ScalarParameter() {defaultValue=128, group=BoxShape, parameterName=U2-Edge}
  n19 = Clamp(None=Multiply(A=Subtract(A=@n16, B=@n17), B=@n18))
  n20 = ComponentMask(None=Rotator(Coordinate=@n10, Time=ScalarParameter() {defaultValue=0.0331, group=BoxShape, parameterName=V1-Rotation}) {speed=6.2832}) {g=true}
  n21 = ScalarParameter() {defaultValue=0.19, group=BoxShape, parameterName=V1-Offset-Inner}
  n22 = ScalarParameter() {defaultValue=128, group=BoxShape, parameterName=V1-Edge}
  n23 = Clamp(None=Multiply(A=Subtract(A=@n20, B=@n21), B=@n22))
  n24 = ScalarParameter() {defaultValue=0.025, group=BoxShape, parameterName=V2-Rotation}
  n25 = OneMinus(None=ComponentMask(None=Rotator(Coordinate=@n10, Time=@n24) {speed=6.2832}) {g=true})
  n26 = ScalarParameter() {defaultValue=0.19, group=BoxShape, parameterName=Y2-Offset-Inner}
  n27 = ScalarParameter() {defaultValue=128, group=BoxShape, parameterName=Y2-Edge}
  n28 = Clamp(None=Multiply(A=Subtract(A=@n25, B=@n26), B=@n27))
  n29 = DotProduct(A=Clamp(None=DotProduct(A=@n14, B=@n19)), B=Clamp(None=DotProduct(A=@n23, B=@n28)))
  n30 = LinearInterpolate(
   A=StaticSwitchParameter(True=Constant(), False=StaticSwitchParameter(True=@n9, False=@n29) {parameterName=IsPip}) {parameterName=IsBar}
   B=Multiply(A=TextureSample(UVs=@n3.UVs) {texture=/Game/Athena/UI/Challenges/Art/MissionTileImages/T_UI_Icon_Mission}.R)
   Alpha=ScalarParameter() {parameterName=IsStar}
  )
  LinearInterpolate(
   A=@n1
   B=StaticSwitchParameter(True=@n2, False=StaticSwitchParameter(True=@n2, False=@n6) {parameterName=IsPip}) {parameterName=IsBar}
   Alpha=@n30
  )
  n31 = VectorParameter() {parameterName=Accent, defaultValue=<rgba 1 0.4023 0.1599 1>}
  n32 = VectorParameter() {parameterName=PrestigeAccent, defaultValue=<rgba 1 0.855 0.0331 1>}
  n33 = ScalarParameter() {parameterName=IsPrestige}
  n34 = VectorParameter() {parameterName=InactiveAccent, defaultValue=<rgba 0.3185 0.6038 1 1>}
  n35 = VectorParameter() {parameterName=PrestigeInactiveAccent, defaultValue=<rgba 0.9734 1 0.4452 1>}
  n36 = ScalarParameter() {parameterName=IsPrestige}
  n37 = LinearInterpolate(
   A=LinearInterpolate(
    A=LinearInterpolate(A=@n31.A, B=@n32.A, Alpha=@n33)
    B=LinearInterpolate(A=@n34.A, B=@n35.A, Alpha=@n36)
    Alpha=ScalarParameter() {parameterName=IsInactive}
   )
   B=LinearInterpolate(
    A=VectorParameter() {parameterName=Primary, defaultValue=<rgba 0.2344 0.113 0.056 1>}.A
    B=VectorParameter() {parameterName=PrimaryInactive, defaultValue=<rgba 0 0.004 0.0723 1>}.A
    Alpha=ScalarParameter() {parameterName=IsInactive}
   )
   Alpha=@n30
  ) {constAlpha=1}
  n38 = Constant() {r=1}
  n39 = LinearInterpolate(B=@n4.A, Alpha=@n30)
  n40 = Abs(None=Add(A=TextureCoordinate()) {constB=-0.5})
  n41 = Multiply(A=Subtract(A=ComponentMask(None=@n40) {g=true}, B=ScalarParameter() {defaultValue=0.125, parameterName=BarThickness})) {constB=256}
  n42 = ScalarParameter() {defaultValue=0.1524, parameterName=Circle1}
  n43 = Clamp(None=Multiply(A=Subtract(A=@n8.RadialGradientExponential, B=@n42), B=ScalarParameter() {defaultValue=166, parameterName=Circle1_Sharp}))
  n44 = LinearInterpolate(A=Clamp(None=OneMinus(None=@n41)), Alpha=@n43) {constB=0}
  n45 = ScalarParameter() {defaultValue=0.125, group=BoxShape, parameterName=V1-Offset-Outer}
  n46 = Clamp(None=Multiply(A=Subtract(A=@n20, B=@n45), B=@n22))
  n47 = ScalarParameter() {defaultValue=0.125, group=BoxShape, parameterName=Y2-Offset-Outer}
  n48 = Clamp(None=Multiply(A=Subtract(A=@n25, B=@n47), B=@n27))
  n49 = ScalarParameter() {defaultValue=0.125, group=BoxShape, parameterName=U1-Offset-Outer}
  n50 = Clamp(None=Multiply(A=Subtract(A=@n11, B=@n49), B=@n13))
  n51 = ScalarParameter() {defaultValue=0.125, group=BoxShape, parameterName=U2-Offset-Outer}
  n52 = Clamp(None=Multiply(A=Subtract(A=@n16, B=@n51), B=@n18))
  n53 = DotProduct(A=Clamp(None=DotProduct(A=@n50, B=@n52)), B=Clamp(None=DotProduct(A=@n46, B=@n48)))
  n54 = StaticSwitchParameter(True=@n43, False=@n53) {parameterName=IsPip}
  LinearInterpolate(
   B=LinearInterpolate(A=LinearInterpolate(A=@n37, B=@n38, Alpha=@n39), B=@n37, Alpha=ScalarParameter() {parameterName=IsDefault})
   Alpha=LinearInterpolate(A=StaticSwitchParameter(True=@n44, False=@n54) {parameterName=IsBar}, B=@n53, Alpha=ScalarParameter() {parameterName=IsStar})
  )
 ]
}
